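; BP_BrickSoupSequenceRunner is a child sequence runner that adds a "bricksoup" console-command event to load and play its sequence for testing.
; The event stores a 0–1 percentage for skipping ahead, then on the server sets a starting method that bypasses some calendar logic, flags the cheat run and starts pre-loading.
; When pre-loading completes, it calls the parent and, on the server for cheat runs only, loads and plays the sequence.
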(event Custom|bricksoup
  (params (Percentage :type Float%20%28double-precision%29))
  ;[0 0 0] Console Command / Load and Play
  (Variables|Cheats|SetCheat%20Day%20Percentage Percentage) ; Allows skipping ahead in the\nsequence for testing. A value\nbetween 0 - 1.0
  (if (Networking|IsServer)
    (Variables|Cheats|SetStarting%20Method "NewEnumerator1") ; Keeps some calendar logic from running
    (Variables|Default|SetRun%20From%20Cheat true)
    (CallFunction|%28CallOnServer%29Pre-LoadSequence)))

(event EventPre-LoadSequenceComplete
  ;[0 0 0] Console Command / Load and Play
  (|CallParentFunction)
  (if (Networking|IsServer)
    (if (Variables|Default|GetRun%20From%20Cheat)
      (CallFunction|LoadSequence) ; Play the sequence.
    )))

(fn ConstructionScript
  (|CallParentFunction))
