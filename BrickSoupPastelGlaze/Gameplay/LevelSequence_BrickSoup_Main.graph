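# Master Level Sequence for BrickSoup gameplay in the BrickSoupPastelGlaze content, lasting under a second.
# It has no bindings of its own and only layers six subsequence tracks, each with one section.
# Two sections start at the beginning, one ending early and the other running throughout.
# The other four begin in staggered fashion around the short section's end and run to the shared finish.
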LevelSequence /BrickSoupPastelGlaze/Gameplay/LevelSequence_BrickSoup_Main
bindings []
display_rate {denominator=1, numerator=60}
tick_resolution {denominator=1, numerator=24000}
tracks [
 {type=MovieSceneSubTrack, display_name=Subsequences, sections [{type=MovieSceneSubSection, end_frame=16200, start_frame=4500}]}
 {type=MovieSceneSubTrack, display_name=Subsequences, sections [{type=MovieSceneSubSection, end_frame=16200, start_frame=5400}]}
 {type=MovieSceneSubTrack, display_name=Subsequences, sections [{type=MovieSceneSubSection, end_frame=16200, start_frame=5400}]}
 {type=MovieSceneSubTrack, display_name=Subsequences, sections [{type=MovieSceneSubSection, end_frame=16200, start_frame=0}]}
 {type=MovieSceneSubTrack, display_name=Subsequences, sections [{type=MovieSceneSubSection, end_frame=5706, start_frame=0}]}
 {type=MovieSceneSubTrack, display_name=Subsequences, sections [{type=MovieSceneSubSection, end_frame=16200, start_frame=5400}]}
]
